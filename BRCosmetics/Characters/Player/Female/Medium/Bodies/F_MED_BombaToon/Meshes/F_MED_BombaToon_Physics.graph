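# Physics asset for the F_MED_BombaToon medium female player body cosmetic, setting up secondary motion for its shirt bones.
# Eight shirt points (three per side plus front and back center) each pair a tiny-capsule base bone with a larger, offset dyn capsule.
# Each dyn bone is constrained only to its own base bone, with twist locked and narrow swing limits that vary by position.
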PhysicsAsset /BRCosmetics/Characters/Player/Female/Medium/Bodies/F_MED_BombaToon/Meshes/F_MED_BombaToon_Physics
bodies {
 base_shirt01_lf [{shapeType=Capsule, length=0.5, radius=0.5, shapeName=base_shirt01_lf_capsule}]
 base_shirt01_rt [{shapeType=Capsule, length=0.5, radius=0.5, shapeName=base_shirt01_rt_capsule}]
 base_shirt02_lf [{shapeType=Capsule, length=0.5, radius=0.5, shapeName=base_shirt02_lf_capsule}]
 base_shirt02_rt [{shapeType=Capsule, length=0.5, radius=0.5, shapeName=base_shirt02_rt_capsule}]
 base_shirt03_lf [{shapeType=Capsule, length=0.5, radius=0.5, shapeName=base_shirt03_lf_capsule}]
 base_shirt03_rt [{shapeType=Capsule, length=0.5, radius=0.5, shapeName=base_shirt03_rt_capsule}]
 base_shirtBack_cr [{shapeType=Capsule, length=0.5, radius=0.5, shapeName=base_shirtBack_cr_capsule}]
 base_shirtFront_cr [{shapeType=Capsule, length=0.5, radius=0.5, shapeName=base_shirtFront_cr_capsule}]
 dyn_shirt01_lf [{shapeType=Capsule, length=3, radius=2.25, shapeName=dyn_shirtBack_cr_capsule, center=<4.409 0 0>, rotation=<rot 90 0 0>}]
 dyn_shirt01_rt [{shapeType=Capsule, length=3, radius=2.25, shapeName=dyn_shirtBack_cr_capsule, center=<4.409 0 0>, rotation=<rot 90 0 0>}]
 dyn_shirt02_lf [{shapeType=Capsule, length=3, radius=2.25, shapeName=dyn_shirtBack_cr_capsule, center=<4.409 0 0>, rotation=<rot 90 0 0>}]
 dyn_shirt02_rt [{shapeType=Capsule, length=3, radius=2.25, shapeName=dyn_shirtBack_cr_capsule, center=<4.409 0 0>, rotation=<rot 90 0 0>}]
 dyn_shirt03_lf [{shapeType=Capsule, length=3, radius=2.25, shapeName=dyn_shirtBack_cr_capsule, center=<4.409 0 0>, rotation=<rot 90 0 0>}]
 dyn_shirt03_rt [{shapeType=Capsule, length=3, radius=2.25, shapeName=dyn_shirtBack_cr_capsule, center=<4.409 0 0>, rotation=<rot 90 0 0>}]
 dyn_shirtBack_cr [{shapeType=Capsule, length=3, radius=2.25, shapeName=dyn_shirtBack_cr_capsule, center=<4.409 0 0>, rotation=<rot 90 0 0>}]
 dyn_shirtFront_cr [{shapeType=Capsule, length=3, radius=2.25, shapeName=dyn_shirtBack_cr_capsule, center=<4.409 0 0>, rotation=<rot 90 0 0>}]
}
constraints [
 {bone1Name=dyn_shirt01_lf, bone2Name=base_shirt01_lf, swing1LimitDegrees=20, swing2LimitDegrees=15, twistLimitDegrees=45, twistMotion=Locked}
 {bone1Name=dyn_shirt01_rt, bone2Name=base_shirt01_rt, swing1LimitDegrees=20, swing2LimitDegrees=15, twistLimitDegrees=45, twistMotion=Locked}
 {bone1Name=dyn_shirt02_lf, bone2Name=base_shirt02_lf, swing1LimitDegrees=5, swing2LimitDegrees=15, twistLimitDegrees=45, twistMotion=Locked}
 {bone1Name=dyn_shirt02_rt, bone2Name=base_shirt02_rt, swing1LimitDegrees=5, swing2LimitDegrees=15, twistLimitDegrees=45, twistMotion=Locked}
 {bone1Name=dyn_shirt03_lf, bone2Name=base_shirt03_lf, swing1LimitDegrees=8, swing2LimitDegrees=15, twistLimitDegrees=45, twistMotion=Locked}
 {bone1Name=dyn_shirt03_rt, bone2Name=base_shirt03_rt, swing1LimitDegrees=8, swing2LimitDegrees=15, twistLimitDegrees=45, twistMotion=Locked}
 {
  bone1Name=dyn_shirtBack_cr
  bone2Name=base_shirtBack_cr
  swing1LimitDegrees=5
  swing2LimitDegrees=15
  twistLimitDegrees=45
  twistMotion=Locked
 }
 {
  bone1Name=dyn_shirtFront_cr
  bone2Name=base_shirtFront_cr
  swing1LimitDegrees=5
  swing2LimitDegrees=15
  twistLimitDegrees=45
  twistMotion=Locked
 }
]
