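# Post-process style secondary-motion AnimBP for the F_MED_BombaToon female body, layering rig and physics corrections onto the pose copied from the parent mesh.
# Cached stages run Control Rig, weighted constraints on two right arm-band base bones, RigidBody simulation in spine_04 space, then spring-driven AnimDynamics on both arm bands.
# Each update stores gravity and RigidBody settings from a parameter asset, plus a zipline-dependent AnimDynamics scale, though no shown node reads these variables.
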AnimBlueprint /BRCosmetics/Characters/Player/Female/Medium/Bodies/F_MED_BombaToon/Meshes/F_MED_BombaToon_AnimBP
target_skeleton=&/F_MED_BombaToon_Skeleton
graphs [
 AnimGraph {nodes [
   Output Pose(Result=Use cached pose 'Anim Dyn Pose'()) {layerGroup=None}
   n1 = Control Rig(Source=Copy Pose From Mesh() {bUseAttachedParent=true}) {controlRigAssetReference {blueprintRigClass=&/F_MED_BombaToon_ControlRig.F_MED_BombaToon_ControlRig_C}}
   Save cached pose 'Ctrl Rig Pose'(Pose=@n1) {cacheName=Ctrl Rig Pose}
   n2 = "Constraint - base_armBand01_rt (PPPP)"!(
    ComponentPose=Local To Component(LocalPose=Use cached pose 'Ctrl Rig Pose'())
    Alpha=1
    ConstraintWeights_0=0.6124
    ConstraintWeights_1:float=0.131
    ConstraintWeights_2:float=0.1292
    ConstraintWeights_3:float=0.1274
   ) -> (Pose) {
    boneToModify {boneName=base_armBand01_rt}
    constraintSetup [
     {targetBone {boneName=upperarm_twist_01_r}}
     {targetBone {boneName=upperarm_twist_02_r}}
     {targetBone {boneName=dyn_armBand01_rt}}
     {targetBone {boneName=upperarm_r}}
    ]
    constraintWeights [1,1,1,1]
   }
   n3 = "Constraint - base_armBand02_rt (PPPP)"!(
    ComponentPose=@n2
    Alpha=1
    ConstraintWeights_0=0.5726
    ConstraintWeights_1:float=0.2005
    ConstraintWeights_2:float=0.1944
    ConstraintWeights_3:float=0.0325
   ) -> (Pose) {
    boneToModify {boneName=base_armBand02_rt}
    constraintSetup [
     {targetBone {boneName=upperarm_twist_01_r}}
     {targetBone {boneName=upperarm_r}}
     {targetBone {boneName=upperarm_twist_02_r}}
     {targetBone {boneName=dyn_armBand02_rt}}
    ]
    constraintWeights [1,1,1,1]
   }
   Save cached pose 'Surf Con Pose'(Pose=Component To Local(ComponentPose=@n3)) {cacheName=Surf Con Pose}
   n4 = Anim Dynamics - Bone: dyn_armBand01_rt!(
    ComponentPose=Local To Component(LocalPose=Use cached pose 'RBAN Pose'())
    Alpha=1
    ComponentLinearAccScale:Vector="0.2,0.2,0.2"
    ComponentLinearVelScale:Vector="0.2,0.2,0.2"
    ExternalForce="0,0,0"
   ) -> (Pose) {
    angularDampingOverride=1
    angularSpringConstant=4000
    bAngularSpring=true
    bOverrideAngularDamping=true
    bUseGravityOverride=true
    gravityScale=0.35
    simulationSpace=RootRelative
    worldSpaceMinimumScale=0.2
    boundBone {boneName=dyn_armBand01_rt}
    componentAppliedLinearAccClamp=<20000 20000 20000>
    componentLinearAccScale=<0.2 0.2 0.2>
    componentLinearVelScale=<0.2 0.2 0.2>
    gravityOverride {z=-200}
    physicsBodyDefinitions [{
      boundBone {boneName=dyn_armBand01_rt}
      boxExtents=<27 17 17>
      constraintSetup {angularLimitsMax=<0 50 50>, angularLimitsMin=<0 -50 -50>, angularTarget=<1 0 0>}
      localJointOffset=<2.5 0 0>
     }]
    relativeSpaceBone {boneName=upperarm_twist_01_r}
    simSpaceSettings {maxAngularAcceleration=150, maxAngularVelocity=20, simSpaceAngularAlpha=0.65}
   }
   n5 = Anim Dynamics - Bone: dyn_armBand02_rt!(
    ComponentPose=@n4
    Alpha=1
    ComponentLinearAccScale:Vector="0.2,0.2,0.2"
    ComponentLinearVelScale:Vector="0.2,0.2,0.2"
    ExternalForce="0,0,0"
   ) -> (Pose) {
    angularDampingOverride=1
    angularSpringConstant=4000
    bAngularSpring=true
    bOverrideAngularDamping=true
    bUseGravityOverride=true
    gravityScale=0.35
    simulationSpace=RootRelative
    worldSpaceMinimumScale=0.2
    boundBone {boneName=dyn_armBand02_rt}
    componentAppliedLinearAccClamp=<20000 20000 20000>
    componentLinearAccScale=<0.2 0.2 0.2>
    componentLinearVelScale=<0.2 0.2 0.2>
    gravityOverride {z=-200}
    physicsBodyDefinitions [{
      boundBone {boneName=dyn_armBand02_rt}
      boxExtents=<27 17 17>
      constraintSetup {angularLimitsMax=<0 50 50>, angularLimitsMin=<0 -50 -50>, angularTarget=<1 0 0>}
      localJointOffset=<2.5 0 0>
     }]
    relativeSpaceBone {boneName=upperarm_twist_01_r}
    simSpaceSettings {maxAngularAcceleration=150, maxAngularVelocity=20, simSpaceAngularAlpha=0.65}
   }
   Save cached pose 'Anim Dyn Pose'(Pose=Component To Local(ComponentPose=@n5)) {cacheName=Anim Dyn Pose}
   n6 = RigidBody!(
    ComponentPose=Local To Component(LocalPose=Use cached pose 'Surf Con Pose'())
    Alpha
    OverrideWorldGravity:Vector="0,0,0"
    ExternalForce
    SimSpaceSettings:Sim Space Settings Structure=(WorldAlpha=0,VelocityScaleZ=1,DampingAlpha=1,MaxLinearVelocity=10000,MaxAngularVelocity=10000,MaxLinearAcceleration=10000,MaxAngularAcceleration=10000,ExternalLinearDragV=(X=0,Y=0,Z=0),ExternalLinearVelocity=(X=0,Y=0,Z=0),ExternalAngularVelocity=(X=0,Y=0,Z=0))
   ) -> (Pose) {
    bOverrideWorldGravity=true
    overridePhysicsAsset=&/F_MED_BombaToon_Physics
    simulationSpace=BaseBoneSpace
    baseBoneRef {boneName=spine_04}
   }
   Save cached pose 'RBAN Pose'(Pose=Component To Local(ComponentPose=@n6)) {cacheName=RBAN Pose}
  ]}
 EventGraph {nodes [
   Event BlueprintUpdateAnimation()
   > n1 = "Gravity Override State Function (Parameter Asset) (Nativized)"(
    CharacterPartAnimInstance=Self-Reference()
    GameTimeInSeconds=GetGameTimeInSeconds()
    Parameters=&/F_MED_BombaToon_Parameters
   )
   > Set GravityOverride(GravityOverride=@n1)
   > Set RBAN Settings Out(RBAN Settings Out=@n1.OutRigidBodyAnimNodeParameters)
   n2 = GetOwnerAnimationInterface!(self:Custom Character Part Anim Instance Object Reference=Self-Reference()) -> (ReturnValue:Fort Player Animation Interface Object Reference)
   > Is Valid(InputObject=@n2)
   > n3 = Set isZipline(isZipline:bool=Break Fort Anim Input Zipline!(FortAnimInput_Zipline:"Fort Anim Input Zipline Structure (by ref)"=GetZiplineInput!(self:Fort Player Animation Interface Object Reference=@n2) -> (ReturnValue:"Fort Anim Input Zipline Structure (by ref)")) -> (
     bIsZiplining:bool=~
     bIsOnAscender:bool=~
     bShouldPlayPivotTransition:bool=~
     LeanYaw:float=~
     PivotBlendDelayRemaining:float=~
     PivotBlendDelay:float=~
     PivotCardinalDirection:EFortCardinalDirection Enum
     WorldVelocityLastTick:Vector=~
    ) {is=title K2Node_BreakStruct}) -> (Output_Get:bool)
   > n4 = Branch(Condition=@n3)
   > Set animDyn on zipline(animDyn on zipline:Vector="0, 0, 0") -> (Output_Get:Vector)
   Set animDyn on zipline(execute=@n4.else, animDyn on zipline="0.2,0.2,0.2")
  ]}
]
variables [
 animDyn on zipline {type=Vector, default=<0.2 0.2 0.2>}
 GravityOverride {type=Vector}
 isZipline {type=boolean}
 RBAN Settings Out {type=FortRigidBodyAnimNodeParameters}
]
