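# Control Rig for the BombaToon character's face accessory part, adjusting the ponytail base bone to follow head motion.
# Its DeltaRotation function converts one swing or twist Euler axis of a relative rotation into a clamped, scaled weight.
# Head rotation against head_null, parented to spine_05, weights the first rotation offset on base_ponytail01_cr.
# A second, opposite offset runs at constant full weight, leaving its paired DeltaRotation output unconnected.
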ControlRigBlueprint /BRCosmetics/Characters/Player/Female/Medium/Bodies/F_MED_BombaToon/Meshes/Parts/F_MED_BombaToon_FaceAcc_CtrlRig
graphs [
 New Function_ContainedGraph {nodes [
   n1 = RigVMFunctionEntryNode()
   RigVMFunction_MathBoolToInteger = &RigVMFunction_MathBoolToInteger(Value=@n1.onlyTwist)
   RigVMFunction_MathQuaternionSwingTwist = &RigVMFunction_MathQuaternionSwingTwist(
    twistAxis=@n1.twistAxis
    Input=&RigUnit_GetRelativeTransformForItem(
     bChildInitial=False
     bParentInitial=False
     child=(Type=Bone,Name="None")
     parent=(Type=Bone,Name="None")
     child.Name=@n1.child
     parent.Name=@n1.parent
     parent.Type=@n1.parentType
    ).RelativeTransform.Rotation
   )
   RigVMFunction_MathQuaternionToEuler = &RigVMFunction_MathQuaternionToEuler(
    Value=RigVMDispatchNode(
     Values=((),())
     Index=@RigVMFunction_MathBoolToInteger.Result
     Values.0=@RigVMFunction_MathQuaternionSwingTwist.Swing
     Values.1=@RigVMFunction_MathQuaternionSwingTwist.Twist
    ) {template="DISPATCH_RigVMDispatch_SelectInt32(in Index,in Values,out Result)"}.Result
    RotationOrder=@n1.Rotation Order
   )
   > DISPATCH_RigVMDispatch_Print = RigVMDispatchNode(ScreenColor=(R=1,G=1,B=1,A=1), ScreenDuration=1.5, Enabled=@n1.debugRotation, Value=@RigVMFunction_MathQuaternionToEuler.Result) {template="DISPATCH_RigVMDispatch_Print(in Prefix,in Value,in Enabled,in ScreenDuration,in ScreenColor)"}
   Clamp = &RigVMFunction_MathIntClamp(Value=@n1.primAxis)
   DISPATCH_RigVMDispatch_SelectInt32 = RigVMDispatchNode(
    Values=(0,0,0)
    Index=@Clamp.Result
    Values.0=@RigVMFunction_MathQuaternionToEuler.Result.X
    Values.1=@RigVMFunction_MathQuaternionToEuler.Result.Y
    Values.2=@RigVMFunction_MathQuaternionToEuler.Result.Z
   ) {template="DISPATCH_RigVMDispatch_SelectInt32(in Index,in Values,out Result)"}
   Remap = &RigVMFunction_MathDoubleRemap(
    TargetMaximum=1
    TargetMinimum=0
    bClamp=False
    Value=@DISPATCH_RigVMDispatch_SelectInt32.Result
    SourceMaximum=@n1.max
    SourceMinimum=@n1.min
   )
   Multiply = &RigVMFunction_MathDoubleMul(A=&RigVMFunction_MathDoubleClamp(Value=@Remap.Result).Result, B=@n1.premult)
   > RigVMFunctionReturnNode(outputWeight=@Multiply.Result)
  ]}
 RigVMFunctionLibrary {nodes [DeltaRotation = RigVMCollapseNode(ExecuteContext=(), Rotation Order=ZYX, premult=1, twistAxis=(X=1,Y=0,Z=0))]}
 RigVMModel {nodes [
   RigUnit_BeginExecution = &RigUnit_BeginExecution()
   RigVMFunction_Sequence = &RigVMFunction_Sequence(ExecuteContext=@RigUnit_BeginExecution.ExecutePin)
   DeltaRotation = RigVMFunctionReferenceNode(
    Rotation Order=ZYX
    child=head
    debugRotation=false
    max=-40
    onlyTwist=false
    parent=head_null
    parentType=Null
    premult=1
    primAxis=2
    twistAxis=(X=1,Y=0,Z=0)
    ExecuteContext=@RigVMFunction_Sequence.A
   )
   > &RigUnit_ModifyTransforms(
    ItemToModify=((Item=(Type=Bone,Name="base_ponytail01_cr"),Transform=(Rotation=(X=0,Y=-0.3827,Z=0,W=0.9239),Translation=(X=0,Y=0,Z=0),Scale3D=(X=1,Y=1,Z=1))))
    WeightMinimum=0
    Weight=@DeltaRotation.outputWeight
   )
   > DeltaRotation_1 = RigVMFunctionReferenceNode(
    Rotation Order=ZYX
    child=head
    debugRotation=false
    max=20
    onlyTwist=false
    parent=head_null
    parentType=Null
    premult=1
    primAxis=2
    twistAxis=(X=1,Y=0,Z=0)
   )
   > &RigUnit_ModifyTransforms(
    ItemToModify=((Item=(Type=Bone,Name="base_ponytail01_cr"),Transform=(Rotation=(X=0,Y=0.0872,Z=0,W=0.9962),Translation=(X=0,Y=0,Z=0),Scale3D=(X=1,Y=1,Z=1))))
    Weight=1
    WeightMinimum=0
   )
  ]}
]
hierarchy [
 root {loc=<0 0 0>, rot=<0 0 0 1>}
 pelvis {parent=root, loc=<0 -0.5301 98.0571>, rot=<0.0127 0.707 -0.0127 -0.707>}
 spine_01 {parent=pelvis, loc=<6.1208 -0.9328 0>, rot=<0 0 0.034 -0.9994>}
 spine_02 {parent=spine_01, loc=<6.5997 -0.2095 0>, rot=<0 0 0.0238 -0.9997>}
 spine_03 {parent=spine_02, loc=<6.4914 -0.1237 0>, rot=<0 0 0.0149 -0.9999>}
 spine_04 {parent=spine_03, loc=<6.9598 0.0794 0>, rot=<0 0 -0.0132 -0.9999>}
 spine_05 {parent=spine_04, loc=<6.9987 0 0>, rot=<0 0 -0.0015 -1>}
 neck_01 {parent=spine_05, loc=<11.3138 0.1529 0>, rot=<0 0 -0.2003 -0.9797>}
 neck_02 {parent=neck_01, loc=<4.5627 0 0>, rot=<0 0 0.0504 -0.9987>}
 head {parent=neck_02, loc=<5.3914 0 0>, rot=<0 0 0.115 -0.9934>}
 base_sunglasses_cr {parent=head, loc=<9.4305 -7.3133 0>, rot=<-0.6421 -0.2961 -0.2961 -0.6421>}
 dyn_sunglasses_cr {parent=base_sunglasses_cr, loc=<0 0 0>, rot=<0 0 0 1>}
 base_hairTieRoot_cr {parent=head, loc=<14.4204 -15.8614 -0.0185>, rot=<0.5825 -0.3574 0.4005 -0.6104>}
 base_ponytail01_cr {parent=base_hairTieRoot_cr, loc=<4.4599 0 0.0001>, rot=<0.063 -0.4488 0.0023 -0.8914>}
 dyn_ponytail01_cr {parent=base_ponytail01_cr, loc=<0 0 0>, rot=<0 0 0 1>}
 dyn_ponytail02_cr {parent=dyn_ponytail01_cr, loc=<8.4201 0 -0.0005>, rot=<-0.0283 -0.6037 -0.0637 -0.7941>}
 dyn_ponytail03_cr {parent=dyn_ponytail02_cr, loc=<8.5063 0.0336 -0.0469>, rot=<0.0122 -0.0148 -0.0484 -0.9986>}
 dyn_ponytail04_cr {parent=dyn_ponytail03_cr, loc=<8.544 0 0.0001>, rot=<-0.1844 0.4333 0.216 -0.8553>}
 base_ponyEnd_lf {parent=dyn_ponytail03_cr, loc=<5.7798 -3.5985 0.5969>, rot=<-0.4468 0.4349 0.2136 -0.7521>}
 dyn_ponyEnd_lf {parent=base_ponyEnd_lf, loc=<0 0 0>, rot=<0 0 0 1>}
 base_ponyEnd_bk {parent=dyn_ponytail03_cr, loc=<8.9704 1.2442 -2.1546>, rot=<0.9912 -0.0085 0.0528 -0.1213>}
 dyn_ponyEnd_bk {parent=base_ponyEnd_bk, loc=<0 0 0>, rot=<0 0 0 1>}
 base_ponyEnd_rt {parent=dyn_ponytail02_cr, loc=<5.3429 5.9437 -0.0914>, rot=<0.6406 0.256 -0.4485 -0.5683>}
 dyn_ponyEnd_rt {parent=base_ponyEnd_rt, loc=<0 0 0>, rot=<0 0 0 1>}
 base_ponyEnd_fwd {parent=dyn_ponytail02_cr, loc=<0.9558 1.2331 5.9466>, rot=<0.0626 -0.0311 0.1315 -0.9888>}
 dyn_ponyEnd_fwd {parent=base_ponyEnd_fwd, loc=<0 0 0>, rot=<0 0 0 1>}
 base_bangsMain01_lf {parent=head, loc=<15.1886 10.847 -7.6841>, rot=<0.2172 -0.9239 0.0696 -0.3073>}
 dyn_bangsMain01_lf {parent=base_bangsMain01_lf, loc=<0 0 0>, rot=<0 0 0 1>}
 dyn_bangsMain02_lf {parent=dyn_bangsMain01_lf, loc=<9.587 -0.0001 0.0002>, rot=<-0.3303 0.183 0.0654 -0.9236>}
 base_bangsSub_lf {parent=dyn_bangsMain01_lf, loc=<3.9678 1.5046 3.535>, rot=<-0.182 0.3098 -0.0541 -0.9317>}
 dyn_bangsSub_lf {parent=base_bangsSub_lf, loc=<0 0 0>, rot=<0 0 0 1>}
 clavicle_l {parent=spine_05, loc=<5.4831 -0.1053 -3.2946>, rot=<0.0465 -0.7289 -0.009 -0.6829>}
 upperarm_l {parent=clavicle_l, loc=<9.8658 -1.2136 -0.4105>, rot=<0.0164 -0.3245 -0.0269 -0.9454>}
 clavicle_r {parent=spine_05, loc=<5.4831 -0.1053 3.2946>, rot=<0.7289 0.0465 -0.6829 0.009>}
 upperarm_r {parent=clavicle_r, loc=<-9.8658 1.2136 0.4105>, rot=<0.0164 -0.3245 -0.0269 -0.9454>}
 head_null {parent=spine_05, type=Null, loc=<0 0 0>, rot=<0 0 0 1>}
]
