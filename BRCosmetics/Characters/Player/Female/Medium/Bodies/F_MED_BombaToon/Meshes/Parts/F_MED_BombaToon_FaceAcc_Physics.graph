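# Physics asset for the face accessory mesh of the female medium BombaToon player cosmetic, setting up a physically simulated ponytail.
# Capsule bodies cover a ponytail base and a chain of four dynamic ponytail bones, plus the head and both upper arms, which likely serve as colliders.
# Constraints link each ponytail bone to the one before it, starting at the base, with twist locked and per-joint swing limits that are tightest at the tip.
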PhysicsAsset /BRCosmetics/Characters/Player/Female/Medium/Bodies/F_MED_BombaToon/Meshes/Parts/F_MED_BombaToon_FaceAcc_Physics
bodies {
 base_ponytail01_cr [{shapeType=Capsule, length=1, radius=1}]
 dyn_ponytail01_cr [{shapeType=Capsule, length=3.5, radius=5, center=<3.3053 0 3.643>, rotation=<rot 0 0 -90>}]
 dyn_ponytail02_cr [{shapeType=Capsule, length=4.5, radius=4.4, center=<4.0493 0 0.1391>, rotation=<rot 0 0 90>}]
 dyn_ponytail03_cr [{shapeType=Capsule, length=1, radius=4.2, center=<4.3335 -0.3153 -0.894>, rotation=<rot 0 0 90>}]
 dyn_ponytail04_cr [{shapeType=Capsule, length=4.5, radius=3.5, center=<6.5368 0 -0.7135>, rotation=<rot 90 0 0>}]
 head [{shapeType=Capsule, length=3.5, radius=14.8392, shapeName=head_capsule, center=<9.9029 -3.0791 -0.2548>, rotation=<rot 90 0 0>}]
 upperarm_l [{shapeType=Capsule, length=14, radius=5, shapeName=clavicle_l_capsule, center=<7.2959 -0.1706 -0.5435>, rotation=<rot 90 0 0>}]
 upperarm_r [{shapeType=Capsule, length=14, radius=5, shapeName=clavicle_l_capsule, center=<-7.2959 -0.1706 -0.5435>, rotation=<rot 90 0 0>}]
}
constraints [
 {
  bone1Name=dyn_ponytail02_cr
  bone2Name=dyn_ponytail01_cr
  swing1LimitDegrees=50
  swing2LimitDegrees=25
  twistLimitDegrees=45
  twistMotion=Locked
 }
 {
  bone1Name=dyn_ponytail03_cr
  bone2Name=dyn_ponytail02_cr
  swing1LimitDegrees=40
  swing2LimitDegrees=35
  twistLimitDegrees=45
  twistMotion=Locked
 }
 {
  bone1Name=dyn_ponytail04_cr
  bone2Name=dyn_ponytail03_cr
  swing1LimitDegrees=15
  swing2LimitDegrees=15
  twistLimitDegrees=45
  twistMotion=Locked
 }
 {
  bone1Name=dyn_ponytail01_cr
  bone2Name=base_ponytail01_cr
  swing1LimitDegrees=14
  swing2LimitDegrees=30
  twistLimitDegrees=45
  twistMotion=Locked
 }
]
