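# Character part AnimBP for F_MED_BombaToon face accessories, adding secondary physics motion to ponytail, bangs and sunglasses.
# Blends parent and body-mesh copied poses at clavicles and upper arms, runs a Control Rig, then caches the result.
# That pose feeds a RigidBody node, then chained bone-relative AnimDynamics nodes.
# The event graph caches the body mesh, derives gravity and rigid-body settings from a parameter asset, and zeroes `animDyn on zipline` while ziplining.
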AnimBlueprint /BRCosmetics/Characters/Player/Female/Medium/Bodies/F_MED_BombaToon/Meshes/Parts/F_MED_BombaToon_FaceAcc_AnimBP
target_skeleton=&/F_MED_BombaToon_FaceAcc_Skeleton
graphs [
 AnimGraph {nodes [
   n1 = RigidBody!(
    ComponentPose=Local To Component(LocalPose=Use cached pose 'Start Pose'())
    Alpha
    OverrideWorldGravity:Vector="0,0,0"
    ExternalForce
    SimSpaceSettings:Sim Space Settings Structure=(WorldAlpha=0.7,VelocityScaleZ=1,DampingAlpha=1,MaxLinearVelocity=850,MaxAngularVelocity=850,MaxLinearAcceleration=1000,MaxAngularAcceleration=1000,ExternalLinearDragV=(X=0,Y=0,Z=0),ExternalLinearVelocity=(X=0,Y=0,Z=0),ExternalAngularVelocity=(X=0,Y=0,Z=0))
   ) -> (Pose) {
    bOverrideWorldGravity=true
    overridePhysicsAsset=&/F_MED_BombaToon_FaceAcc_Physics
    simulationSpace=BaseBoneSpace
    baseBoneRef {boneName=base_ponytail01_cr}
    simSpaceSettings {maxAngularAcceleration=1000, maxAngularVelocity=850, maxLinearAcceleration=1000, maxLinearVelocity=850, worldAlpha=0.7}
   }
   n2 = Anim Dynamics - Bone: base_sunglasses_cr(ComponentPose=@n1, Alpha=1, ExternalForce="0,0,0") {
    angularDampingOverride=0.9999
    angularSpringConstant=800
    bOverrideAngularDamping=true
    bUseGravityOverride=true
    gravityScale=0.35
    linearSpringConstant=10000
    simulationSpace=BoneRelative
    worldSpaceMinimumScale=0
    boundBone {boneName=base_sunglasses_cr}
    componentAppliedLinearAccClamp=<3500 3500 3500>
    componentLinearAccScale=<0.35 0.35 0.35>
    componentLinearVelScale=<0.35 0.35 0.35>
    gravityOverride {z=-700}
    physicsBodyDefinitions [{
      boundBone {boneName=base_sunglasses_cr}
      boxExtents=<40 40 30>
      constraintSetup {angularLimitsMax=<0 0 0>, angularLimitsMin=<0 -20 0>, angularTarget=<1 0 0>}
      localJointOffset=<3 0 0>
     }]
    relativeSpaceBone {boneName=dyn_sunglasses_cr}
    simSpaceSettings {maxAngularAcceleration=4, maxAngularVelocity=6}
   }
   n3 = Anim Dynamics - Bone: dyn_bangsMain01_lf!(ComponentPose=@n2, Alpha=1, ComponentLinearAccScale:Vector="1,1,1", ComponentLinearVelScale:Vector="1,1,1", ExternalForce="0,0,0") -> (Pose) {
    angularDampingOverride=1
    angularSpringConstant=4500
    bAngularSpring=true
    bOverrideAngularDamping=true
    bUseGravityOverride=true
    gravityScale=0.35
    simulationSpace=BoneRelative
    worldSpaceMinimumScale=0
    boundBone {boneName=dyn_bangsMain01_lf}
    componentAppliedLinearAccClamp=<2000 2000 2000>
    componentLinearAccScale=<1 1 1>
    componentLinearVelScale=<1 1 1>
    gravityOverride {z=-100}
    physicsBodyDefinitions [{
      boundBone {boneName=dyn_bangsMain01_lf}
      boxExtents=<33 33 33>
      constraintSetup {angularLimitsMax=<0 5 0>, angularLimitsMin=<0 -10 0>, angularTarget=<1 0 0>}
      localJointOffset=<2 0 0>
     }]
    relativeSpaceBone {boneName=base_bangsMain01_lf}
    simSpaceSettings {maxAngularAcceleration=18, maxAngularVelocity=8, simSpaceAngularAlpha=0.4}
   }
   n4 = Anim Dynamics - Bone: dyn_bangsMain01_lf(ComponentPose=@n3, Alpha=1, ComponentLinearAccScale="1,1,1", ComponentLinearVelScale="1,1,1", ExternalForce="0,0,0") {
    angularDampingOverride=1
    angularSpringConstant=4500
    bAngularSpring=true
    bOverrideAngularDamping=true
    bUseGravityOverride=true
    gravityScale=0.35
    simulationSpace=BoneRelative
    worldSpaceMinimumScale=0
    boundBone {boneName=dyn_bangsMain01_lf}
    componentAppliedLinearAccClamp=<2000 2000 2000>
    componentLinearAccScale=<1 1 1>
    componentLinearVelScale=<1 1 1>
    gravityOverride {z=-100}
    physicsBodyDefinitions [{
      boundBone {boneName=dyn_bangsMain01_lf}
      boxExtents=<25 25 25>
      constraintSetup {angularLimitsMax=<0 18 8>, angularLimitsMin=<0 -18 -15>, angularTarget=<1 0 0>}
      localJointOffset=<2 0 0>
     }]
    relativeSpaceBone {boneName=dyn_bangsMain01_lf}
    simSpaceSettings {maxAngularAcceleration=18, maxAngularVelocity=8, simSpaceAngularAlpha=0.4}
   }
   n5 = Anim Dynamics - Bone: dyn_bangsSub_lf!(ComponentPose=@n4, Alpha=1, ComponentLinearAccScale:Vector="1,1,1", ComponentLinearVelScale:Vector="1,1,1", ExternalForce="0,0,0") -> (Pose) {
    angularDampingOverride=1
    angularSpringConstant=4000
    bAngularSpring=true
    bOverrideAngularDamping=true
    bUseGravityOverride=true
    gravityScale=0.35
    simulationSpace=BoneRelative
    worldSpaceMinimumScale=0
    boundBone {boneName=dyn_bangsSub_lf}
    componentAppliedLinearAccClamp=<2000 2000 2000>
    componentLinearAccScale=<1 1 1>
    componentLinearVelScale=<1 1 1>
    gravityOverride {z=-100}
    physicsBodyDefinitions [{
      boundBone {boneName=dyn_bangsSub_lf}
      boxExtents=<15 15 15>
      constraintSetup {angularLimitsMax=<0 15 25>, angularLimitsMin=<0 -15 -25>, angularTarget=<1 0 0>}
      localJointOffset=<2 0 0>
     }]
    relativeSpaceBone {boneName=dyn_bangsMain01_lf}
    simSpaceSettings {maxAngularAcceleration=20, maxAngularVelocity=10, simSpaceAngularAlpha=0.25}
   }
   n6 = Anim Dynamics - Bone: dyn_ponyEnd_lf!(ComponentPose=@n5, Alpha=1, ComponentLinearAccScale:Vector="1,1,1", ComponentLinearVelScale:Vector="1,1,1", ExternalForce="0,0,0") -> (Pose) {
    angularDampingOverride=1
    angularSpringConstant=4000
    bAngularSpring=true
    bOverrideAngularDamping=true
    bUseGravityOverride=true
    gravityScale=0.35
    simulationSpace=BoneRelative
    worldSpaceMinimumScale=0
    boundBone {boneName=dyn_ponyEnd_lf}
    componentAppliedLinearAccClamp=<20000 20000 20000>
    componentLinearAccScale=<1 1 1>
    componentLinearVelScale=<1 1 1>
    gravityOverride {z=-100}
    physicsBodyDefinitions [{
      boundBone {boneName=dyn_ponyEnd_lf}
      boxExtents=<20 20 20>
      constraintSetup {angularLimitsMax=<0 10 10>, angularLimitsMin=<0 -10 -10>, angularTarget=<1 0 0>}
      localJointOffset=<3 0 0>
     }]
    relativeSpaceBone {boneName=base_ponyEnd_lf}
    simSpaceSettings {maxAngularAcceleration=20, maxAngularVelocity=10, simSpaceAngularAlpha=0.25}
   }
   n7 = Anim Dynamics - Bone: dyn_ponyEnd_rt!(ComponentPose=@n6, Alpha=1, ComponentLinearAccScale:Vector="1,1,1", ComponentLinearVelScale:Vector="1,1,1", ExternalForce="0,0,0") -> (Pose) {
    angularDampingOverride=1
    angularSpringConstant=4000
    bAngularSpring=true
    bOverrideAngularDamping=true
    bUseGravityOverride=true
    gravityScale=0.35
    simulationSpace=BoneRelative
    worldSpaceMinimumScale=0
    boundBone {boneName=dyn_ponyEnd_rt}
    componentAppliedLinearAccClamp=<20000 20000 20000>
    componentLinearAccScale=<1 1 1>
    componentLinearVelScale=<1 1 1>
    gravityOverride {z=-100}
    physicsBodyDefinitions [{
      boundBone {boneName=dyn_ponyEnd_rt}
      boxExtents=<25 25 25>
      constraintSetup {angularLimitsMax=<0 10 10>, angularLimitsMin=<0 -10 -10>, angularTarget=<1 0 0>}
      localJointOffset=<3 0 0>
     }]
    relativeSpaceBone {boneName=base_ponyEnd_rt}
    simSpaceSettings {maxAngularAcceleration=20, maxAngularVelocity=10, simSpaceAngularAlpha=0.25}
   }
   n8 = Anim Dynamics - Bone: dyn_ponyEnd_fwd!(ComponentPose=@n7, Alpha=1, ComponentLinearAccScale:Vector="1,1,1", ComponentLinearVelScale:Vector="1,1,1", ExternalForce="0,0,0") -> (Pose) {
    angularDampingOverride=1
    angularSpringConstant=4000
    bAngularSpring=true
    bOverrideAngularDamping=true
    bUseGravityOverride=true
    gravityScale=0.35
    simulationSpace=BoneRelative
    worldSpaceMinimumScale=0
    boundBone {boneName=dyn_ponyEnd_fwd}
    componentAppliedLinearAccClamp=<20000 20000 20000>
    componentLinearAccScale=<1 1 1>
    componentLinearVelScale=<1 1 1>
    gravityOverride {z=-100}
    physicsBodyDefinitions [{
      boundBone {boneName=dyn_ponyEnd_fwd}
      boxExtents=<20 20 20>
      constraintSetup {angularLimitsMax=<0 7 10>, angularLimitsMin=<0 -10 -10>, angularTarget=<1 0 0>}
      localJointOffset=<3 0 0>
     }]
    relativeSpaceBone {boneName=base_ponyEnd_fwd}
    simSpaceSettings {maxAngularAcceleration=20, maxAngularVelocity=10, simSpaceAngularAlpha=0.25}
   }
   n9 = Anim Dynamics - Bone: dyn_ponyEnd_bk!(ComponentPose=@n8, Alpha=1, ComponentLinearAccScale:Vector="1,1,1", ComponentLinearVelScale:Vector="1,1,1", ExternalForce="0,0,0") -> (Pose) {
    angularDampingOverride=1
    angularSpringConstant=4000
    bAngularSpring=true
    bOverrideAngularDamping=true
    bUseGravityOverride=true
    gravityScale=0.35
    simulationSpace=BoneRelative
    worldSpaceMinimumScale=0
    boundBone {boneName=dyn_ponyEnd_bk}
    componentAppliedLinearAccClamp=<20000 20000 20000>
    componentLinearAccScale=<1 1 1>
    componentLinearVelScale=<1 1 1>
    gravityOverride {z=-100}
    physicsBodyDefinitions [{
      boundBone {boneName=dyn_ponyEnd_bk}
      boxExtents=<25 25 25>
      constraintSetup {angularLimitsMax=<0 10 10>, angularLimitsMin=<0 -10 -10>, angularTarget=<1 0 0>}
      localJointOffset=<3 0 0>
     }]
    relativeSpaceBone {boneName=base_ponyEnd_bk}
    simSpaceSettings {maxAngularAcceleration=20, maxAngularVelocity=10, simSpaceAngularAlpha=0.25}
   }
   Output Pose(Result=Component To Local(ComponentPose=@n9)) {layerGroup=None}
   n10 = Layered blend per bone(
    BasePose=Copy Pose From Mesh() {bUseAttachedParent=true}
    BlendPoses_0=Copy Pose From Mesh(SourceMeshComponent=Get Body_Mesh()) {bUseAttachedParent=true}
   ) {layerSetup [{branchFilters [{boneName=upperarm_l}, {boneName=upperarm_r}, {boneName=clavicle_l}, {boneName=clavicle_r}]}]}
   n11 = Control Rig(Source=@n10) {controlRigAssetReference {blueprintRigClass=&/F_MED_BombaToon_FaceAcc_CtrlRig.F_MED_BombaToon_FaceAcc_CtrlRig_C}}
   Save cached pose 'Start Pose'(Pose=@n11) {cacheName=Start Pose}
  ]}
 EventGraph {nodes [
   Event BlueprintUpdateAnimation()
   > n1 = "Gravity Override State Function (Parameter Asset) (Nativized)"(
    CharacterPartAnimInstance=Self-Reference()
    GameTimeInSeconds=GetGameTimeInSeconds()
    Parameters=&/F_MED_BombaToon_FaceAcc_Parameters
   )
   > Set GravityOverride(GravityOverride=@n1)
   > Set RBAN Settings Out(RBAN Settings Out=@n1.OutRigidBodyAnimNodeParameters)
   n2 = GetOwnerAnimationInterface!(self:Custom Character Part Anim Instance Object Reference=Self-Reference()) -> (ReturnValue:Fort Player Animation Interface Object Reference)
   > Is Valid(InputObject=@n2)
   > n3 = Set isZipline(isZipline:bool=Break Fort Anim Input Zipline!(FortAnimInput_Zipline:"Fort Anim Input Zipline Structure (by ref)"=GetZiplineInput!(self:Fort Player Animation Interface Object Reference=@n2) -> (ReturnValue:"Fort Anim Input Zipline Structure (by ref)")) -> (
     bIsZiplining:bool=~
     bIsOnAscender:bool=~
     bShouldPlayPivotTransition:bool=~
     LeanYaw:float=~
     PivotBlendDelayRemaining:float=~
     PivotBlendDelay:float=~
     PivotCardinalDirection:EFortCardinalDirection Enum
     WorldVelocityLastTick:Vector=~
    ) {is=title K2Node_BreakStruct}) -> (Output_Get:bool)
   > n4 = Branch(Condition=@n3)
   > Set animDyn on zipline(animDyn on zipline:Vector="0, 0, 0") -> (Output_Get:Vector)
   Set animDyn on zipline(execute=@n4.else, animDyn on zipline="0.25,0.25,0.25")
   Event OnCharacterCustomizationLoaded()
   > n5 = Cast To PlayerPawn_Athena(Object=TryGetPawnOwner())
   > Set Body_Mesh(Body_Mesh=GetSkeletalMeshForPartType(self=@n5))
  ]}
]
variables [
 animDyn on zipline {type=Vector, default=<0.25 0.25 0.25>}
 Body_Mesh {type=SkeletalMeshComponent}
 GravityOverride {type=Vector}
 isZipline {type=boolean}
 RBAN Settings Out {type=FortRigidBodyAnimNodeParameters}
]
